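; Blueprint for the LabRat Satellite A building prop, which tints its satellite mesh from the building's texture data.
; The construction script calls its parent, fetches the building texture data array and checks that the entry at index 1 is valid.
; If so, it converts that entry's tint color to a linear color and passes it to SetColoronSecondaryMesh for SM_LabRat_Satellite_A.
; The BeginPlay, overlap and Tick events are empty.
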(event EventBeginPlay)

(event Collision|EventActorBeginOverlap
  (params (OtherActor :type Actor%20Object%20Reference)))

(event EventTick
  (params (DeltaSeconds :type Float%20%28single-precision%29)))

(fn ConstructionScript
  (|CallParentFunction)
  (bind _outtexturedataarray (Building|GetBuildingTextureData))
  (Utilities|IsValid (Utilities|Array|Get(acopy) _outtexturedataarray 1)
    (:Is%20Valid
      (Default|SetColoronSecondaryMesh (Math|Conversions|ToLinearColor(Color) (Variables|/Script/FortniteGame.BuildingTextureData|GetTintColor (Utilities|Array|Get(acopy) _outtexturedataarray 1))) (Variables|Default|GetSM_LabRat_Satellite_A)))))
